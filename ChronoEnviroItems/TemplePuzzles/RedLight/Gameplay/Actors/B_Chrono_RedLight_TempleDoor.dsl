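; Two-panel sliding temple vault door for the RedLight puzzle, opened by player interaction or by a dropped pickup that meets a required rarity.
; BeginPlay stores panel start/end positions along forward vectors, and timelines lerp the panels between them, feeding alpha to a Niagara dust effect.
; Replicated open/destroyed states drive collision and visibility, while authority switches keep sound, camera shake and FX on remotes.
; A debug reset flag auto-closes the door after a delay.
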(event EventBeginPlay
  ;[1 1 1] Init
  (Default|SetStartandEndPositionsforDoors))

(event Custom|Open
  ;[1 1 1] Open Door
  (if (not (Variables|Default|GetDebugResetState))
    (Utilities|FlowControl|DoOnce false
      (:Completed
        _)))
  (Variables|Default|SetDoorOpened true)
  (Utilities|FlowControl|Sequence
    (:then_0
      (CallFunction|DoorAnimatingCosmetics) ; inc audio
      (if (Variables|Default|GetDebugResetState)
        (Utilities|FlowControl|Delay 1e+01)
        (CallFunction|CloseVaultDoor)))
    (:then_1
      (bind (_direction _dooralpha) (Timeline|DoorOpenAnim PlayFromStart)
        (:Update
          (Niagara|SetNiagaraVariable(Float) (Variables|Default|GetNS_Dust) "DoorOpen" _dooralpha)
          (Transformation|SetWorldLocation (Variables|Default|GetSM_Ancients_VaultDoor_R) (Math|Vector|Lerp(Vector) (Variables|Default|GetStartLocation_R) (Variables|Default|GetEndLocation_R) _dooralpha) false false)
          (Transformation|SetWorldLocation (Variables|Default|GetSM_Ancients_VaultDoor_L) (Math|Vector|Lerp(Vector) (Variables|Default|GetStartLocation_L) (Variables|Default|GetEndLocation_L) _dooralpha) false false))
        (:EndShake
          (Camera|PlayWorldCameraShake "/Game/Abilities/Player/Commando/Actives/FragGrenade/B_GrenadeCameraShake.B_GrenadeCameraShake_C" (Transformation|GetWorldLocation (Variables|Default|GetStaticMesh)) 2e+03 2e+03 0.75))))))

(event Custom|OpenVaultDoor
  ;[1 1 1] Open Door
  (if (not (Variables|Default|GetDebugResetState))
    (Utilities|FlowControl|DoOnce false
      (:Completed
        _)))
  (Variables|Default|SetDoorOpened true)
  (Utilities|FlowControl|Sequence
    (:then_0
      (CallFunction|DoorAnimatingCosmetics) ; inc audio
      (if (Variables|Default|GetDebugResetState)
        (Utilities|FlowControl|Delay 1e+01)
        (CallFunction|CloseVaultDoor)))
    (:then_1
      (Timeline|DoorOpenAnim PlayFromStart))))

(event Custom|CloseVaultDoor
  ;[1 1 1] Open Door
  (if (Variables|Default|GetDebugResetState)
    (Variables|Default|SetDoorOpened false)
    (else
      (Utilities|FlowControl|DoOnce false
        (:Completed
          (Variables|Default|SetDoorOpened true)))))
  (CallFunction|DoorAnimatingCosmetics)
  (bind (_direction _dooralpha) (Timeline|Timeline PlayFromStart)
    (:Update
      (Niagara|SetNiagaraVariable(Float) (Variables|Default|GetNS_Dust) "DoorOpen" _dooralpha)
      (Transformation|SetWorldLocation (Variables|Default|GetSM_Ancients_VaultDoor_R) (Math|Vector|Lerp(Vector) (Variables|Default|GetEndLocation_R) (Variables|Default|GetStartLocation_R) _dooralpha) false false)
      (Transformation|SetWorldLocation (Variables|Default|GetSM_Ancients_VaultDoor_L) (Math|Vector|Lerp(Vector) (Variables|Default|GetEndLocation_L) (Variables|Default|GetStartLocation_L) _dooralpha) false false))) ; door closing anim
)

(event Custom|DoorAnimatingCosmetics
  ;[1 1 1] FX and audio
  (Networking|SwitchHasAuthority
    (:Remote
      (Audio|SpawnSoundatLocation "/Game/Sounds/Fort_Traps/Dungeon_Traps/Dungeon_Door_Open_Cue.Dungeon_Door_Open_Cue" (Transformation|GetWorldLocation (Variables|Default|GetStaticMesh)))
      (Camera|PlayWorldCameraShake "/Game/Abilities/Player/Commando/Actives/FragGrenade/B_GrenadeCameraShake.B_GrenadeCameraShake_C" (Transformation|GetWorldLocation (Variables|Default|GetStaticMesh)) 2e+03 2e+03 0.75)
      (bind _returnvalue (Niagara|SpawnSystemAttached "/ChronoEnviroItems/TemplePuzzles/RedLight/FX/NS_Chrono_Temple_DoorDust.NS_Chrono_Temple_DoorDust" (Variables|Default|GetStaticMesh) "None" (Transformation|GetWorldLocation (Variables|Default|GetStaticMesh)) "0, 0, 0" "KeepWorldPosition"))
      (Variables|Default|SetNS_Dust _returnvalue))))

(event Custom|DestroyDoor
  (bind _self self)
  ;[1 1 1] Destroy
  (Variables|Default|SetDoorDestroyed true)
  (Networking|SwitchHasAuthority
    (:Authority
      (Utilities|IsValid _self
        (:Is%20Valid
          (Actor|DestroyActor _self))))
    (:Remote
      (Niagara|SpawnSystematLocation "/Game/Effects/Fort_Effects/World/Destruction/Base_Effects/NS_General_Deco_Destruction_Large_Converted.NS_General_Deco_Destruction_Large_Converted" (Transformation|GetWorldLocation (Variables|Default|GetDefaultSceneRoot)))
      (Audio|SpawnSoundatLocation "/Game/Sounds/Fort_Impact_Sounds/Demolition/Defaults/Demo_Default_Rock_Impact_Final_Burst_Cue.Demo_Default_Rock_Impact_Final_Burst_Cue" (Transformation|GetWorldLocation (Variables|Default|GetDefaultSceneRoot)))
      (Rendering|SetActorHiddenInGame self true))))

(event Interaction|EventBlueprintOnInteract
  (params (InteractingPawn :type Fort%20Pawn%20Object%20Reference) (InteractionBeingAttempted :type EInteractionBeingAttempted%20Enum :default "FirstInteraction"))
  (CallFunction|OpenVaultDoor)
  (Variables|Default|SetCanInteract false)
  (Variables|Interaction|SetbShowFirstInteractPrompt false))

(event Custom|MultiCastCheckRarityAndDoTorchCosmtetics
  (params (NewParam :type Actor%20Object%20Reference))
  (bind _rarity_map_from_enum_to_integer_value (Variables|Default|GetRarity%20Map%20From%20Enum%20to%20Integer%20Value))
  ;[1 1 1] Proto - drop object of rarity, door opens
  (bind _asfort_pickup_athena (Utilities|Casting|CastToFortPickupAthena NewParam))
  (Variables|Default|SetRarity%20Of%20Item%20Checked%20Int (Utilities|Map|Find _rarity_map_from_enum_to_integer_value (Class|FortItemFunctionLibrary|GetRarity (Pickup|GetPrimaryItemDefinition _asfort_pickup_athena))))
  (if (>= (Utilities|Map|Find _rarity_map_from_enum_to_integer_value (Class|FortItemFunctionLibrary|GetRarity (Pickup|GetPrimaryItemDefinition _asfort_pickup_athena))) (Utilities|Map|Find (Variables|Default|GetRarity%20Map%20From%20Enum%20to%20Integer%20Value) (Variables|Default|GetRarity%20Check%20Required%20Rarity)))
    (Actor|DestroyActor _asfort_pickup_athena)
    (Utilities|FlowControl|Delay 2.0)
    (if (Variables|Default|GetDestroyItem)
      (Effects|Components|ParticleSystem|SpawnFXAtLocation :FXSystemAsset "/Game/Effects/Fort_Effects/Emotes/Haste/NS_Default_SmokePuff.NS_Default_SmokePuff" :Location (Transformation|GetActorLocation _asfort_pickup_athena) :Scale "2.000000,2.000000,2.000000" :bAutoDestroy true :PoolingMethod "None" :bAutoActivate true)
      (else
        _))
    (CallFunction|OpenVaultDoor)
    (if (Variables|Default|GetDebugResetState)
      (Utilities|FlowControl|Delay 1e+01)
      (CallFunction|CloseVaultDoor))))

(fn BlueprintCanInteract
  (params (InteractingPawn :type Fort%20Pawn%20Object%20Reference) (InteractionBeingAttempted :type EInteractionBeingAttempted%20Enum :default "FirstInteraction") (InteractionType :type TInteractionType%20Enum :default "IT_NoInteraction"))
  (return (Variables|Default|GetCanInteract)))

(fn BlueprintGetFailedInteractionString
  (params (InteractingPawn :type Fort%20Pawn%20Object%20Reference) (InteractionBeingAttempted :type EInteractionBeingAttempted%20Enum :default "FirstInteraction"))
  (return ""))

(fn BlueprintGetInteractionString
  (params (InteractingPawn :type Fort%20Pawn%20Object%20Reference) (InteractionBeingAttempted :type EInteractionBeingAttempted%20Enum :default "FirstInteraction"))
  (return "NSLOCTEXT(\"[E5858784E8A6664271F449FD013C107E]\", \"FB6B58264FD98BA976013B84BA7872CD\", \"Open Vault Door\")"))

(fn MapDoorToAltar
  (params (Hit%20result :type Array%20of%20Hit%20Result%20Structures))
  (for _array_element Hit%20result
    _))

(fn OnRep_CanInteract)

(fn OnRep_DoorDestroyed
  (if (Variables|Default|GetDoorDestroyed)
    (Rendering|SetVisibility (Variables|Default|GetStaticMesh) false false)
    (Collision|SetCollisionEnabled (Variables|Default|GetBlockAllCollision) "NoCollision")
    (else
      (Rendering|SetVisibility (Variables|Default|GetStaticMesh) true false))))

(fn OnRep_DoorOpened
  (if (Variables|Default|GetDoorOpened)
    (Collision|SetCollisionEnabled (Variables|Default|GetBlockAllCollision) "NoCollision")
    (else
      (Collision|SetCollisionEnabled (Variables|Default|GetBlockAllCollision) "QueryAndPhysics"))))

(fn SetStartAndEndPositionsForDoors
  (bind _sm_ancients_vaultdoor_l (Variables|Default|GetSM_Ancients_VaultDoor_L))
  (bind _returnvalue (Transformation|GetWorldLocation _sm_ancients_vaultdoor_l))
  (bind _sm_ancients_vaultdoor_r (Variables|Default|GetSM_Ancients_VaultDoor_R))
  (bind _returnvalue_2 (Transformation|GetWorldLocation _sm_ancients_vaultdoor_r))
  ;[1 1 1] Store off base 'start' position. Right door needs to go 'backwards' negative units
  (Variables|Default|SetStartLocation_R _returnvalue_2)
  (Variables|Default|SetEndLocation_R (+ _returnvalue_2 (* (Transformation|GetForwardVector _sm_ancients_vaultdoor_r) (Math|Conversions|ToVector(Float) (Variables|Default|GetNegativeDoorExtensionLength)))))
  ;[1 1 1] Store off base 'start' position. Left door needs to go 'forwards' positive units
  (Variables|Default|SetStartLocation_L _returnvalue)
  (Variables|Default|SetEndLocation_L (+ _returnvalue (* (Transformation|GetForwardVector _sm_ancients_vaultdoor_l) (Math|Conversions|ToVector(Float) (Variables|Default|GetDoorExtensionLength)))))
  (return))

(fn ConstructionScript)
