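; Red-light/green-light temple statue that turns on a timer and rift-teleports players caught moving inside its detection box.
; A looping delay chain plays audio, pauses as a telegraph, rotates the mesh, checks velocity, then rotates back.
; Overlaps maintain the player array; tagged pawns are exempt, and only positive velocity components count as movement.
; Caught players trigger a cue and flash, are stopped, then force-teleported by a server-spawned rift, a commented hack.
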(event EventTick
  (params (DeltaSeconds :type Float%20%28single-precision%29))
  ;[1 1 1] TODO: Can we do this more event / non-tick based? Or optimize by checking against time delta before array check
  (if (Variables|Rotation|GetFacing%20Detection)
    _
    (elif (> DeltaSeconds (Variables|Default|GetTick%20Delta%20Time%20Passed%20Requirement))
      (CallFunction|CheckPlayerVelocitySpawnRiftandStopPlayerMovement)
      (else
        (Variables|Default|SetTick%20Delta%20Time%20Passed%20Requirement (+ (Variables|Default|GetTime%20Add%20to%20Server%20Time%20To%20Check) (Fortnite|Sync|Time|GetServerWorldTimeSeconds)))))))

(event EventBeginPlay
  (bind _staticmesh1 (Variables|Default|GetStaticMesh1))
  (bind _returnvalue (Transformation|GetWorldRotation _staticmesh1))
  ;[1 1 1] Store off facing rotation
  (Variables|Rotation|SetStartAndDetectionAngles _returnvalue)
  (Utilities|FlowControl|Delay 1.0)
  ;[1 1 1] Store off angles to rotate to
  (Transformation|SetWorldRotation _staticmesh1 (CallFunction|AdjustRotationBack) false false)
  (Variables|Rotation|SetResetAngles (Math|Rotator|MakeRotator (.roll _returnvalue) (.pitch _returnvalue) (Math|Float|SelectFloat (- (.yaw _returnvalue) (Variables|Rotation|GetRotationAmount)) (+ (.yaw _returnvalue) (Variables|Rotation|GetRotationAmount)) (> (.yaw _returnvalue) 1.8e+02))))
  (CallFunction|RedLightGreenLightTimer))

(event Custom|StartRedLightTimer
  (CallFunction|RedLightGreenLightTimer))

(event Custom|RotateActorForDetectionState
  (Transformation|SetWorldRotation (Variables|Default|GetStaticMesh1) (CallFunction|AdjustRotation) false false)
  (Variables|Rotation|SetFacing%20Detection true))

(event Custom|CheckPlayerVelocitySpawnRiftAndStopPlayerMovement
  (for _array_element (Variables|Teleport|GetPlayerArray)
    (bind _asfort_player_pawn (Utilities|Casting|CastToFortPlayerPawn _array_element))
    (if (GameplayTags|HasAnyMatchingGameplayTags _asfort_player_pawn (Variables|Default|GetAvoid%20Detection%20Tags%20Container))
      _
      (elif (Math|Boolean|ORBoolean (> (.x (Transformation|GetVelocity _asfort_player_pawn)) 0) (> (.y (Transformation|GetVelocity _asfort_player_pawn)) 0) (> (.z (Transformation|GetVelocity _asfort_player_pawn)) 0))
        ;[1 0.9625 0] Play Detection GC. Note/TODO: DoOnce with reset does not seem to be working....
        (Utilities|FlowControl|Sequence
          (:then_0
            (if (Variables|Detection|GetDo%20Caught%20GC)
              (Variables|Detection|SetDo%20Caught%20GC false)
              (GameplayCue|ExecuteGameplayCue self (Variables|Detection|GetCaught%20Gameplay%20Cue%20Tag))
              (Components|Activation|Activate (Variables|Default|GetNS_S17_Mindwipe_Flash) false)))
          (:then_1
            ;[1 1 1] Timer to allow caught player to experience caught GC
            (Utilities|FlowControl|Delay (Variables|Detection|GetDetection%20Allow%20Audio%20To%20Play%20Time))
            ;[1 1 1] To help make sure the player gets hit by the rift
            (Components|Movement|StopMovementImmediately (Variables|/Script/Engine.Character|GetCharacterMovement _asfort_player_pawn))
            (Networking|SwitchHasAuthority
              (:Authority
                ;[1 1 1] Hack: try to force teleport, spawning rifts on moving player is not great...
                (bind (_returnvalue_location _returnvalue_rotation) (Math|Transform|BreakTransform (Transformation|GetWorldTransform (Variables|/Script/Engine.Character|GetMesh _asfort_player_pawn))))
                (bind _returnvalue (Game|SpawnActorfromClass :Class "/Game/Athena/Items/Consumables/RiftItem/BGA_RiftPortal_Item_Athena.BGA_RiftPortal_Item_Athena_C" :SpawnTransform (Math|Transform|MakeTransform _returnvalue_location _returnvalue_rotation) :CollisionHandlingOverride "AdjustIfPossibleButAlwaysSpawn" :TransformScaleMethod "OverrideRootScale"))
                (Class|BGARiftPortalItemAthena|ForceRiftTeleport _returnvalue _asfort_player_pawn)
                (Utilities|FlowControl|Delay (Variables|Teleport|GetRiftLifeTime))
                (Actor|DestroyActor _returnvalue))))))))
  (Variables|Detection|SetDo%20Caught%20GC true))

(event Custom|RedLight%20GreenLight%20Timer
  (Utilities|FlowControl|Sequence
    (:then_0
      (Components|Activation|Activate (Variables|Default|GetLooping%20Audio) false))
    (:then_1
      (Utilities|FlowControl|Delay (Variables|Rotation|GetTime%20Before%20Rotate%20To%20Detection))
      (Utilities|FlowControl|Sequence
        (:then_0
          (Components|Activation|Deactivate (Variables|Default|GetLooping%20Audio))
          (Utilities|FlowControl|Delay (Variables|Detection|GetTeleGraphBeforeTurnDuration))
          (CallFunction|RotateActorforDetectionState)
          (Utilities|FlowControl|Delay (Variables|Detection|GetTeleGraphBeforeTurnDuration))
          (CallFunction|CheckPlayerVelocitySpawnRiftandStopPlayerMovement))
        (:then_1
          (Utilities|FlowControl|Delay (Variables|Rotation|GetTime%20Before%20Rotate%20Back%20to%20Idle))
          (CallFunction|RotateBacktoNonDetectionState)
          (CallFunction|RedLightGreenLightTimer))))))

(event Custom|RotateBackToNonDetectionState
  (Transformation|SetWorldRotation (Variables|Default|GetStaticMesh1) (CallFunction|AdjustRotationBack) false false)
  (Variables|Rotation|SetFacing%20Detection false))

(event OnComponentBeginOverlap(DetectionBox)
  (params (OverlappedComponent :type Primitive%20Component%20Object%20Reference) (OtherActor :type Actor%20Object%20Reference) (OtherComp :type Primitive%20Component%20Object%20Reference) (OtherBodyIndex :type Integer) (bFromSweep :type Boolean) (SweepResult :type Hit%20Result%20Structure%20%28by%20ref%29))
  ;[1 1 1] TODO: On begin overlap, add player to array, remove on end overlap. Rotate statue on timer, on specific angle / timing check, fire event that detects any velocity from any players in array. do sprint detection and immediately snap into place and rift targets
  (bind _asfort_player_pawn (Utilities|Casting|CastToFortPlayerPawn OtherActor))
  (Utilities|Array|AddUnique (Variables|Teleport|GetPlayerArray) _asfort_player_pawn))

(event Custom|InterruptTimerRiftPlayersAndResetTimer
  (CallFunction|CheckPlayerVelocitySpawnRiftandStopPlayerMovement)
  (bind _returnvalue (Utilities|Time|SetTimerbyEvent (EventDispatchers|CreateEvent :!event Custom|StartRedLightTimer) (Variables|Rotation|GetRotateTimerHandleTime) true))
  (bind _output_get (Variables|Rotation|SetRotate%20Timer%20Handle _returnvalue))
  (if (Utilities|Time|IsTimerPausedbyHandle _output_get)
    (Utilities|Time|UnpauseTimerbyHandle _output_get)))

(event OnComponentEndOverlap(DetectionBox)
  (params (OverlappedComponent :type Primitive%20Component%20Object%20Reference) (OtherActor :type Actor%20Object%20Reference) (OtherComp :type Primitive%20Component%20Object%20Reference) (OtherBodyIndex :type Integer))
  ;[1 1 1] TODO: On begin overlap, add player to array, remove on end overlap. Rotate statue on timer, on specific angle / timing check, fire event that detects any velocity from any players in array. do sprint detection and immediately snap into place and rift targets
  (bind _asfort_player_pawn (Utilities|Casting|CastToFortPlayerPawn OtherActor))
  (Utilities|Array|RemoveItem (Variables|Teleport|GetPlayerArray) _asfort_player_pawn))

(fn AdjustRotation
  (return (Math|Rotator|MakeRotator (.roll (Transformation|GetWorldRotation (Variables|Default|GetStaticMesh1))) (.pitch (Transformation|GetWorldRotation (Variables|Default|GetStaticMesh1))) (Math|Float|SelectFloat (- (.yaw (Transformation|GetWorldRotation (Variables|Default|GetStaticMesh1))) (Variables|Rotation|GetRotationAmount)) (+ (.yaw (Transformation|GetWorldRotation (Variables|Default|GetStaticMesh1))) (Variables|Rotation|GetRotationAmount)) (> (.yaw (Transformation|GetWorldRotation (Variables|Default|GetStaticMesh1))) 1.8e+02)))))

(fn AdjustRotationBack
  (return (Math|Rotator|MakeRotator (.roll (Transformation|GetWorldRotation (Variables|Default|GetStaticMesh1))) (.pitch (Transformation|GetWorldRotation (Variables|Default|GetStaticMesh1))) (Math|Float|SelectFloat (- (.yaw (Transformation|GetWorldRotation (Variables|Default|GetStaticMesh1))) (Variables|Rotation|GetRotationAmount)) (+ (.yaw (Transformation|GetWorldRotation (Variables|Default|GetStaticMesh1))) (Variables|Rotation|GetRotationAmount)) (> (.yaw (Transformation|GetWorldRotation (Variables|Default|GetStaticMesh1))) 1.8e+02)))))

(fn ConstructionScript
  (Variables|Rotation|SetStartAndDetectionAngles (Transformation|GetWorldRotation (Variables|Default|GetStaticMesh1))))
